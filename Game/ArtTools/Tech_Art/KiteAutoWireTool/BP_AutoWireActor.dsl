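; Kite Auto Wire actor that strings an animated wire mesh from a ground anchor to a kite or balloon.
; Construction runs UpdateWire: snaps to the ground socket, scales the wire by distance, rotates itself and the kite along it, and binds destruction of both ends.
; Wind, wobble, thickness and a location-seeded random value go into custom primitive data for the materials.
; BeginPlay still calls the deprecated update variant, which writes non-default primitive data.
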(event EventBeginPlay
  (Default|UpdateWireDeprecated))

(event Collision|EventActorBeginOverlap
  (params (OtherActor :type Actor%20Object%20Reference)))

(event EventTick
  (params (DeltaSeconds :type Float%20%28single-precision%29)))

(fn ConvertSelectedActor
  (params (Target :type Actor%20Object%20Reference))
  (bind _returnvalue (Actor|GetComponentbyClass Target "/Script/Engine.StaticMeshComponent"))
  (Utilities|IsValid Target
    (:Is%20Valid
      (Utilities|IsValid _returnvalue
        (:Is%20Valid
          (return _returnvalue true))
        (:Is%20Not%20Valid
          (bind _asstatic_mesh_actor (Utilities|Casting|CastToStaticMeshActor Target)
            (:then
              (return (Variables|/Script/Engine.StaticMeshActor|GetStaticMeshComponent _asstatic_mesh_actor) (Utilities|IsValid (Variables|/Script/Engine.StaticMeshActor|GetStaticMeshComponent _asstatic_mesh_actor) :!class /Script/Engine.KismetSystemLibrary)))
            (:CastFailed
              (return 0 false))))))))

(fn DestructionBindings
  (Utilities|FlowControl|Sequence
    (:then_0
      (bind _asbuilding_static_mesh (Utilities|Casting|CastToBuildingSMActor (Variables|KiteSettings|GetGroundStaticMeshActor)))
      (EventDispatchers|BindEventtoOnDeathPlayEffectsDelegate _asbuilding_static_mesh (EventDispatchers|CreateEvent :!event OnBaseDestroyed)) ; Client Side
      (EventDispatchers|BindEventtoOnDied _asbuilding_static_mesh (EventDispatchers|CreateEvent :!event OnBaseDestroyedServer)) ; Server side
    )
    (:then_1
      (bind _asbuilding_static_mesh_2 (Utilities|Casting|CastToBuildingSMActor (Variables|KiteSettings|GetKiteBalloonMeshActor)))
      (EventDispatchers|BindEventtoOnDeathPlayEffectsDelegate _asbuilding_static_mesh_2 (EventDispatchers|CreateEvent :!event OnKiteBalloonDestroyed)) ; Only client needed as we're only\ndisabling the visibility of the rope
    )))

(fn GetAllValues
  (return (Variables|Default|GetKiteData) (Variables|KiteSettings|GetKiteBalloonMeshActor) (Variables|Default|GetWireMaterial) (Variables|Default|GetGroundWireConnectionMesh) (Variables|Default|GetThickness) (Variables|Default|GetAutoLookAt) (Variables|Default|GetKiteSocketAnchorName) (Variables|Default|GetGroundSocketAnchorName) (Variables|KiteSettings|GetGroundStaticMeshActor) (Variables|Default|GetbInflatable)))

(fn Initialize
  (params (KiteData :type BP%20Kite%20Wire%20Data%20Object%20Reference) (GroundWireConnectionMesh :type Static%20Mesh%20Object%20Reference) (WireMaterial :type Material%20Instance%20Object%20Reference) (Thickness :type Float%20%28double-precision%29 :default 1e+02) (KiteBaloonMeshActor :type Actor%20Object%20Reference) (KiteSocketAnchorName :type Name :default "FX_Death") (GroundSocketAnchorName :type Name :default "FX_Death") (GroundStaticMeshActor :type Actor%20Object%20Reference) (ForceKiteBalloonLookAt :type Boolean) (bInflatable :type Boolean))
  (Utilities|FlowControl|Sequence
    (:then_0
      (Variables|Default|SetGroundWireConnectionMesh GroundWireConnectionMesh))
    (:then_1
      (Variables|Default|SetWireMaterial WireMaterial))
    (:then_2
      (Variables|Default|SetKiteData KiteData))
    (:then_3
      (Variables|Default|SetThickness Thickness))
    (:then_4
      (Variables|KiteSettings|SetKiteBalloonMeshActor KiteBaloonMeshActor))
    (:then_5
      (Variables|Default|SetKiteSocketAnchorName KiteSocketAnchorName))
    (:then_6
      (Variables|Default|SetGroundSocketAnchorName GroundSocketAnchorName))
    (:then_7
      (Variables|KiteSettings|SetGroundStaticMeshActor GroundStaticMeshActor))
    (:then_8
      (Variables|Default|SetAutoLookAt ForceKiteBalloonLookAt))
    (:then_9
      (Variables|Default|SetbInflatable bInflatable))))

(fn OnBaseDestroyed
  (params (Damage :type Float%20%28single-precision%29) (DamageTags :type Gameplay%20Tag%20Container%20Structure%20%28by%20ref%29) (Momentum :type Vector) (HitInfo :type Hit%20Result%20Structure%20%28by%20ref%29) (InstigatedBy :type Fort%20Pawn%20Object%20Reference) (DamageCauser :type Actor%20Object%20Reference) (EffectContext :type Gameplay%20Effect%20Context%20Handle%20Structure))
  (Actor|DestroyActor self)
  (return))

(fn OnBaseDestroyedServer
  (params (DamagedActor :type Actor%20Object%20Reference) (Damage :type Float%20%28single-precision%29) (InstigatedBy :type Controller%20Object%20Reference) (DamageCauser :type Actor%20Object%20Reference) (HitLocation :type Vector) (FHitComponent :type Primitive%20Component%20Object%20Reference) (BoneName :type Name) (Momentum :type Vector))
  (bind _kiteballoonmeshactor (Variables|KiteSettings|GetKiteBalloonMeshActor)
    (:then
      (Actor|DestroyActor _kiteballoonmeshactor))
    (:else
      _))
  (Actor|DestroyActor))

(fn OnKiteBalloonDestroyed
  (params (Damage :type Float%20%28single-precision%29) (DamageTags :type Gameplay%20Tag%20Container%20Structure%20%28by%20ref%29) (Momentum :type Vector) (HitInfo :type Hit%20Result%20Structure%20%28by%20ref%29) (InstigatedBy :type Fort%20Pawn%20Object%20Reference) (DamageCauser :type Actor%20Object%20Reference) (EffectContext :type Gameplay%20Effect%20Context%20Handle%20Structure))
  (bind _wiremeshcomponent (Variables|Default|GetWireMeshComponent))
  (Rendering|SetVisibility _wiremeshcomponent false false)
  (Components|DestroyComponent _wiremeshcomponent)
  (return))

(fn UpdateWire
  (bind _kitedata (Variables|Default|GetKiteData))
  (bind (_wirematerial _thickness) (Class|BPKiteWireData|GetAllValues _kitedata))
  (bind _groundstaticmeshactor (Variables|KiteSettings|GetGroundStaticMeshActor))
  (bind (_returnvalue _isvalid) (CallFunction|ConvertSelectedActor _groundstaticmeshactor))
  (bind _groundconnectionsocketname (Variables|Default|GetGroundConnectionSocketName))
  (bind _groundconnectionbase (Variables|Default|GetGroundConnectionBase))
  (Components|StaticMesh|SetStaticMesh (Variables|Default|GetGroundConnectionBase) (Variables|Default|GetGroundWireConnectionMesh))
  (Utilities|FlowControl|Sequence
    (:then_0
      (if (not (Variables|Default|GetOverrideDataAsset))
        (Variables|Default|SetWireMaterial _wirematerial)
        (Variables|Default|SetThickness _thickness)))
    (:then_1
      (Variables|Default|GetGroundWireConnectionMesh
        (:then
          (if (Transformation|DoesSocketExist _groundconnectionbase _groundconnectionsocketname)
            (Transformation|SetWorldLocation (Variables|Default|GetWireMeshComponent) (Transformation|GetSocketLocation _groundconnectionbase _groundconnectionsocketname) false false)
            (else
              _)))
        (:else
          _))
      (LocalVariables|SetRandomFloat (Math|Random|RandomFloatfromStream (Math|Random|MakeRandomStream (+ (Variables|KiteSettings|GetSeed) (Math|Float|Truncate (* (Class|EnvironmentArtFuncLib|RandomFloatbyLocation (Transformation|GetActorLocation) 5.0 true) 1e+04))))))
      (Utilities|FlowControl|Sequence
        (:then_0
          (CallFunction|DestructionBindings)
          (if (and (Variables|KiteSettings|GetSnapToGroundActorPivot) _isvalid)
            ; Move the actor to the ground position
            (Transformation|SetActorLocation self (Transformation|GetSocketLocation _returnvalue (Variables|Default|GetGroundSocketAnchorName)) true true)
            (else
              _))
          (Transformation|SetWorldScale3D (Variables|Default|GetWireMeshComponent) (Math|Vector|MakeVector (/ (+ (Math|Vector|Distance(Vector) (Transformation|GetActorLocation (Variables|KiteSettings|GetKiteBalloonMeshActor)) (Transformation|GetWorldLocation (Variables|Default|GetWireMeshComponent))) 5.0) 1e+02) (/ (+ (Math|Vector|Distance(Vector) (Transformation|GetActorLocation (Variables|KiteSettings|GetKiteBalloonMeshActor)) (Transformation|GetWorldLocation (Variables|Default|GetWireMeshComponent))) 5.0) 1e+02) (/ (+ (Math|Vector|Distance(Vector) (Transformation|GetActorLocation (Variables|KiteSettings|GetKiteBalloonMeshActor)) (Transformation|GetWorldLocation (Variables|Default|GetWireMeshComponent))) 5.0) 1e+02)))
          ; Just in case the user wants to the Kite/Balloon to look at the rope.
          ;;[0.15 0 0 0.5] Order of opperation is important here
          (Transformation|SetActorRotation (Math|Rotator|MakeRotfromZ (- (Transformation|GetActorLocation (Variables|KiteSettings|GetKiteBalloonMeshActor)) (Transformation|GetActorLocation)))) ; Set initial rotation to get proper up/forward/right axis
          (Transformation|SetActorRotation (Math|Rotator|MakeRotfromXZ (Math|Vector|Lerp(Vector) (* (Transformation|GetActorRightVector) (Math|Float|Sign(Float) (- 1.8e+02 (Math|Float|Absolute(Float) (Variables|KiteSettings|GetRotation))))) (* (Transformation|GetActorForwardVector) (Math|Float|Sign(Float) (Variables|KiteSettings|GetRotation))) (- 1.0 (Math|Float|Absolute(Float) (- (* (Math|Float|Absolute(Float) (/ (Variables|KiteSettings|GetRotation) 3.6e+02)) 2.0) 1.0)))) (Transformation|GetActorUpVector))) ; Set rotation by taking account "Rotation" var
          (Transformation|SetActorRotation (Variables|KiteSettings|GetKiteBalloonMeshActor) (Transformation|GetActorRotation) false) ; Copy rotation to kite
        )
        (:then_1
          (bind _kiteballoonmeshactor (Variables|KiteSettings|GetKiteBalloonMeshActor)
            (:then
              (if (.IsValid (CallFunction|ConvertSelectedActor _kiteballoonmeshactor))
                (Utilities|FlowControl|Sequence
                  (:then_0
                    (bind _returnvalue_2 (Rendering|Material|GetVectorParameterValue (Class|BTechArtGlobalLibrary|GetDynamicMaterialInstance (CallFunction|ConvertSelectedActor _kiteballoonmeshactor)) "Wind Speed" :!class /Script/Engine.MaterialInstanceDynamic))
                    (bind (_returnvalue_r _returnvalue_g _returnvalue_b) (Math|Color|BreakColor _returnvalue_2))
                    (Rendering|Material|SetDefaultCustomPrimitiveDataVector3 (Variables|Default|GetWireMeshComponent) 0 (Math|Vector|MakeVector _returnvalue_r _returnvalue_g _returnvalue_b)))
                  (:then_1
                    (bind _returnvalue_3 (Rendering|Material|GetScalarParameterValue (Class|BTechArtGlobalLibrary|GetDynamicMaterialInstance (CallFunction|ConvertSelectedActor _kiteballoonmeshactor)) "Twist Intensity" :!class /Script/Engine.MaterialInstanceDynamic))
                    (Rendering|Material|SetDefaultCustomPrimitiveDataFloat (Variables|Default|GetWireMeshComponent) 4 _returnvalue_3))
                  (:then_2
                    (bind _returnvalue_4 (Rendering|Material|GetScalarParameterValue (Class|BTechArtGlobalLibrary|GetDynamicMaterialInstance (CallFunction|ConvertSelectedActor _kiteballoonmeshactor)) "Wobble Intensity" :!class /Script/Engine.MaterialInstanceDynamic))
                    (Rendering|Material|SetDefaultCustomPrimitiveDataFloat (Variables|Default|GetWireMeshComponent) 5 _returnvalue_4))
                  (:then_3
                    (bind _returnvalue_5 (Rendering|Material|GetScalarParameterValue (Class|BTechArtGlobalLibrary|GetDynamicMaterialInstance (CallFunction|ConvertSelectedActor _kiteballoonmeshactor)) "Pull Amount" :!class /Script/Engine.MaterialInstanceDynamic))
                    (Rendering|Material|SetDefaultCustomPrimitiveDataFloat (Variables|Default|GetWireMeshComponent) 6 _returnvalue_5))
                  (:then_4
                    (Rendering|Material|SetDefaultCustomPrimitiveDataFloat (Variables|Default|GetWireMeshComponent) 7 (Variables|Default|GetThickness))
                    (Rendering|Material|SetDefaultCustomPrimitiveDataFloat (Variables|Default|GetWireMeshComponent) 14 (LocalVariables|GetRandomFloat))
                    (Rendering|Material|SetDefaultCustomPrimitiveDataFloat (Variables|Default|GetWireMeshComponent) 15 (Math|Conversions|ToFloat(Boolean) (Variables|KiteSettings|GetbAlwaysInTention)))))
                (else
                  (Development|PrintString "Fail to convert To set Wire Data"))))
            (:else
              (Development|PrintString "Kite mesh actor is None To set Wire Data"))))
        (:then_2
          ; This sets values to the Kite or Ballon custom primitive if you want.
          (bind _kiteballoonmeshactor_2 (Variables|KiteSettings|GetKiteBalloonMeshActor)
            (:then
              (if (.IsValid (CallFunction|ConvertSelectedActor _kiteballoonmeshactor_2))
                (Rendering|Material|SetDefaultCustomPrimitiveDataVector3 (CallFunction|ConvertSelectedActor _kiteballoonmeshactor_2) 0 (Transformation|GetWorldLocation (Variables|Default|GetWireMeshComponent)))
                (Rendering|Material|SetDefaultCustomPrimitiveDataFloat (CallFunction|ConvertSelectedActor _kiteballoonmeshactor_2) 14 (LocalVariables|GetRandomFloat))
                (Rendering|Material|SetDefaultCustomPrimitiveDataFloat (CallFunction|ConvertSelectedActor _kiteballoonmeshactor_2) 15 (Math|Conversions|ToFloat(Boolean) (Variables|KiteSettings|GetbAlwaysInTention)))
                ; Kites sometimes turn black, this is to be sure they don't
                (bind _asbp_kites_and_inflatables (Utilities|Casting|CastToBP_Kites_and_Inflatables (Variables|KiteSettings|GetKiteBalloonMeshActor)))
                (bind _outtexturedataarray (Building|GetBuildingTextureData _asbp_kites_and_inflatables :!class /Script/FortniteGame.BuildingSMActor))
                (if (Utilities|Array|IsNotEmpty _outtexturedataarray)
                  (Default|SetColoronSecondaryMesh (Math|Conversions|ToLinearColor(Color) (Variables|/Script/FortniteGame.BuildingTextureData|GetTintColor (Utilities|Array|Get(acopy) _outtexturedataarray))) (Variables|/Script/FortniteGame.BuildingSMActor|GetStaticMeshComponent _asbp_kites_and_inflatables)))
                (else
                  (Development|PrintString "Fail to convert To set Kite Pivot"))))
            (:else)))
        (:then_3
          (bind _wirematerial_2 (Variables|Default|GetWireMaterial)
            (:then
              (Rendering|Material|SetMaterial (Variables|Default|GetWireMeshComponent) 0 _wirematerial_2))
            (:else)))))))

(fn UpdateWire_Deprecated
  (bind _kitedata (Variables|Default|GetKiteData))
  (bind (_wirematerial _thickness) (Class|BPKiteWireData|GetAllValues _kitedata))
  (bind _groundstaticmeshactor (Variables|KiteSettings|GetGroundStaticMeshActor))
  (bind (_returnvalue _isvalid) (CallFunction|ConvertSelectedActor _groundstaticmeshactor))
  (bind _groundconnectionsocketname (Variables|Default|GetGroundConnectionSocketName))
  (bind _groundconnectionbase (Variables|Default|GetGroundConnectionBase))
  (Components|StaticMesh|SetStaticMesh (Variables|Default|GetGroundConnectionBase) (Variables|Default|GetGroundWireConnectionMesh))
  (Utilities|FlowControl|Sequence
    (:then_0
      (if (not (Variables|Default|GetOverrideDataAsset))
        (Variables|Default|SetWireMaterial _wirematerial)
        (Variables|Default|SetThickness _thickness)))
    (:then_1
      (Variables|Default|GetGroundWireConnectionMesh
        (:then
          (if (Transformation|DoesSocketExist _groundconnectionbase _groundconnectionsocketname)
            (Transformation|SetWorldLocation (Variables|Default|GetWireMeshComponent) (Transformation|GetSocketLocation _groundconnectionbase _groundconnectionsocketname) false false)
            (else
              _)))
        (:else
          _))
      (LocalVariables|SetRandomFloat (Math|Random|RandomFloatfromStream (Math|Random|MakeRandomStream (+ (Variables|KiteSettings|GetSeed) (Math|Float|Truncate (* (Class|EnvironmentArtFuncLib|RandomFloatbyLocation (Transformation|GetActorLocation) 5.0 true) 1e+04))))))
      (Utilities|FlowControl|Sequence
        (:then_0
          (CallFunction|DestructionBindings)
          (if (and (Variables|KiteSettings|GetSnapToGroundActorPivot) _isvalid)
            ; Move the actor to the ground position
            (Transformation|SetActorLocation self (Transformation|GetSocketLocation _returnvalue (Variables|Default|GetGroundSocketAnchorName)) true true)
            (else
              _))
          (Transformation|SetWorldScale3D (Variables|Default|GetWireMeshComponent) (Math|Vector|MakeVector (/ (+ (Math|Vector|Distance(Vector) (Transformation|GetActorLocation (Variables|KiteSettings|GetKiteBalloonMeshActor)) (Transformation|GetWorldLocation (Variables|Default|GetWireMeshComponent))) 5.0) 1e+02) (/ (+ (Math|Vector|Distance(Vector) (Transformation|GetActorLocation (Variables|KiteSettings|GetKiteBalloonMeshActor)) (Transformation|GetWorldLocation (Variables|Default|GetWireMeshComponent))) 5.0) 1e+02) (/ (+ (Math|Vector|Distance(Vector) (Transformation|GetActorLocation (Variables|KiteSettings|GetKiteBalloonMeshActor)) (Transformation|GetWorldLocation (Variables|Default|GetWireMeshComponent))) 5.0) 1e+02)))
          ; Just in case the user wants to the Kite/Balloon to look at the rope.
          (if (Variables|Default|GetAutoLookAt)
            ;;[0.15 0 0 0.5] Order of opperation is important here
            (Transformation|SetActorRotation (Math|Rotator|MakeRotfromZ (- (Transformation|GetActorLocation (Variables|KiteSettings|GetKiteBalloonMeshActor)) (Transformation|GetActorLocation)))) ; Set initial rotation to get proper up/forward/right axis
            (Transformation|SetActorRotation (Math|Rotator|MakeRotfromXZ (Math|Vector|Lerp(Vector) (* (Transformation|GetActorRightVector) (Math|Float|Sign(Float) (- 1.8e+02 (Math|Float|Absolute(Float) (Variables|KiteSettings|GetRotation))))) (* (Transformation|GetActorForwardVector) (Math|Float|Sign(Float) (Variables|KiteSettings|GetRotation))) (- 1.0 (Math|Float|Absolute(Float) (- (* (Math|Float|Absolute(Float) (/ (Variables|KiteSettings|GetRotation) 3.6e+02)) 2.0) 1.0)))) (Transformation|GetActorUpVector))) ; Set rotation by taking account "Rotation" var
            (Transformation|SetActorRotation (Variables|KiteSettings|GetKiteBalloonMeshActor) (Transformation|GetActorRotation) false) ; Copy rotation to kite
          ))
        (:then_1
          (bind _kiteballoonmeshactor (Variables|KiteSettings|GetKiteBalloonMeshActor)
            (:then
              (if (.IsValid (CallFunction|ConvertSelectedActor _kiteballoonmeshactor))
                (Utilities|FlowControl|Sequence
                  (:then_0
                    (bind _returnvalue_2 (Rendering|Material|GetVectorParameterValue (Class|BTechArtGlobalLibrary|GetDynamicMaterialInstance (CallFunction|ConvertSelectedActor _kiteballoonmeshactor)) "Wind Speed" :!class /Script/Engine.MaterialInstanceDynamic))
                    (bind (_returnvalue_r _returnvalue_g _returnvalue_b) (Math|Color|BreakColor _returnvalue_2))
                    (Rendering|Material|SetCustomPrimitiveDataVector3 (Variables|Default|GetWireMeshComponent) 0 (Math|Vector|MakeVector _returnvalue_r _returnvalue_g _returnvalue_b)))
                  (:then_1
                    (bind _returnvalue_3 (Rendering|Material|GetScalarParameterValue (Class|BTechArtGlobalLibrary|GetDynamicMaterialInstance (CallFunction|ConvertSelectedActor _kiteballoonmeshactor)) "Twist Intensity" :!class /Script/Engine.MaterialInstanceDynamic))
                    (Rendering|Material|SetCustomPrimitiveDataFloat (Variables|Default|GetWireMeshComponent) 4 _returnvalue_3))
                  (:then_2
                    (bind _returnvalue_4 (Rendering|Material|GetScalarParameterValue (Class|BTechArtGlobalLibrary|GetDynamicMaterialInstance (CallFunction|ConvertSelectedActor _kiteballoonmeshactor)) "Wobble Intensity" :!class /Script/Engine.MaterialInstanceDynamic))
                    (Rendering|Material|SetCustomPrimitiveDataFloat (Variables|Default|GetWireMeshComponent) 5 _returnvalue_4))
                  (:then_3
                    (bind _returnvalue_5 (Rendering|Material|GetScalarParameterValue (Class|BTechArtGlobalLibrary|GetDynamicMaterialInstance (CallFunction|ConvertSelectedActor _kiteballoonmeshactor)) "Pull Amount" :!class /Script/Engine.MaterialInstanceDynamic))
                    (Rendering|Material|SetCustomPrimitiveDataFloat (Variables|Default|GetWireMeshComponent) 6 _returnvalue_5))
                  (:then_4
                    (Rendering|Material|SetCustomPrimitiveDataFloat (Variables|Default|GetWireMeshComponent) 7 (Variables|Default|GetThickness))
                    (Rendering|Material|SetCustomPrimitiveDataFloat (Variables|Default|GetWireMeshComponent) 14 (LocalVariables|GetRandomFloat))
                    (Rendering|Material|SetCustomPrimitiveDataFloat (Variables|Default|GetWireMeshComponent) 15 (Math|Conversions|ToFloat(Boolean) (Variables|KiteSettings|GetbAlwaysInTention)))))
                (else
                  (Development|PrintString "Fail to convert To set Wire Data"))))
            (:else
              (Development|PrintString "Kite mesh actor is None To set Wire Data"))))
        (:then_2
          ; This sets values to the Kite or Ballon custom primitive if you want.
          (bind _kiteballoonmeshactor_2 (Variables|KiteSettings|GetKiteBalloonMeshActor)
            (:then
              (if (.IsValid (CallFunction|ConvertSelectedActor _kiteballoonmeshactor_2))
                (Rendering|Material|SetCustomPrimitiveDataVector3 (CallFunction|ConvertSelectedActor _kiteballoonmeshactor_2) 0 (Transformation|GetWorldLocation (Variables|Default|GetWireMeshComponent)))
                (Rendering|Material|SetCustomPrimitiveDataFloat (CallFunction|ConvertSelectedActor _kiteballoonmeshactor_2) 14 (LocalVariables|GetRandomFloat))
                (Rendering|Material|SetCustomPrimitiveDataFloat (CallFunction|ConvertSelectedActor _kiteballoonmeshactor_2) 15 (Math|Conversions|ToFloat(Boolean) (Variables|KiteSettings|GetbAlwaysInTention)))
                (else
                  (Development|PrintString "Fail to convert To set Kite Pivot"))))
            (:else)))
        (:then_3
          (bind _wirematerial_2 (Variables|Default|GetWireMaterial)
            (:then
              (Rendering|Material|SetMaterial (Variables|Default|GetWireMeshComponent) 0 _wirematerial_2))
            (:else)))))))

(fn ConstructionScript
  (Default|UpdateWire)
  (if (Utilities|IsValid (Variables|KiteSettings|GetGroundStaticMeshActor) :!class /Script/Engine.KismetSystemLibrary)
    (Rendering|Components|Sprite|SetSprite (Variables|Default|GetBillboard) "/Game/Athena/UI/InGame/Creative/PropertyEditor/Icon_Picker/Textures/128x128/T_UI_IconLibrary_anchor_128.T_UI_IconLibrary_anchor_128")))
