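; BP_KiteWireData is a data holder for the Kite auto wire tool, storing a wire's material instance and thickness.
; GetAllValues returns both stored values together.
; UpdateAllValues takes a new material and thickness and writes them to the two variables in sequence, with no validation or other logic.
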(fn GetAllValues
  (return (Variables|Default|GetWireMaterial) (Variables|Default|GetThickness)))

(fn UpdateAllValues
  (params (WireMaterial :type Material%20Instance%20Object%20Reference) (Thickness :type Float%20%28double-precision%29))
  (Utilities|FlowControl|Sequence
    (:then_0
      (Variables|Default|SetWireMaterial WireMaterial))
    (:then_1
      (Variables|Default|SetThickness Thickness))))
