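; Kite and inflatable prop actor that destroys itself when a pawn touches it or when it takes damage.
; A Sphere overlap calls Destroy On Overlap, which destroys the actor if the other actor casts to FortPawn and CanExecuteCosmeticEvents returns false.
; Any server damage event destroys it unconditionally.
; BeginPlay, Tick, ActorBeginOverlap and the construction script are empty.
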(event EventBeginPlay)

(event Collision|EventActorBeginOverlap
  (params (OtherActor :type Actor%20Object%20Reference)))

(event EventTick
  (params (DeltaSeconds :type Float%20%28single-precision%29)))

(event OnComponentBeginOverlap(Sphere)
  (params (OverlappedComponent :type Primitive%20Component%20Object%20Reference) (OtherActor :type Actor%20Object%20Reference) (OtherComp :type Primitive%20Component%20Object%20Reference) (OtherBodyIndex :type Integer) (bFromSweep :type Boolean) (SweepResult :type Hit%20Result%20Structure%20%28by%20ref%29))
  (CallFunction|DestroyonOverlap OtherActor))

(event Custom|Destroy%20On%20Overlap
  (params (Overlapping%20Actor :type Actor%20Object%20Reference))
  (Networking|CanExecuteCosmeticEvents
    (:False
      (Utilities|Casting|CastToFortPawn Overlapping%20Actor)
      (Actor|DestroyActor))))

(event EventOnDamageServer
  (params (Damage :type Float%20%28single-precision%29) (DamageTags :type Gameplay%20Tag%20Container%20Structure%20%28by%20ref%29) (Momentum :type Vector :default "0, 0, 0") (HitInfo :type Hit%20Result%20Structure%20%28by%20ref%29) (InstigatedBy :type Controller%20Object%20Reference) (DamageCauser :type Actor%20Object%20Reference) (EffectContext :type Gameplay%20Effect%20Context%20Handle%20Structure))
  (Actor|DestroyActor))

(fn ConstructionScript)
