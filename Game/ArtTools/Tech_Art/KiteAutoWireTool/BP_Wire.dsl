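; BP_Wire is a construction-time tool that strings a rope mesh from itself to a referenced building prop, likely a kite.
; With a valid prop, it aims Z at it, applies an adjustable roll, copies that rotation to the prop and uniformly scales the rope by distance.
; It copies the prop material's wind, twist, wobble and pull into rope custom primitive data, and writes its own location into the prop's.
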(event EventBeginPlay)

(event Collision|EventActorBeginOverlap
  (params (OtherActor :type Actor%20Object%20Reference)))

(event EventTick
  (params (DeltaSeconds :type Float%20%28single-precision%29)))

(fn ConstructionScript
  (bind _buildingprop (Variables|Default|GetBuildingProp))
  (bind _staticmeshcomponent (Variables|/Script/FortniteGame.BuildingSMActor|GetStaticMeshComponent _buildingprop))
  (bind _material_instance_dynamic (Class|BTechArtGlobalLibrary|GetDynamicMaterialInstance _staticmeshcomponent))
  (bind _sm_rope (Variables|Default|GetSM_Rope))
  (Variables|Default|GetBuildingProp
    (:then
      (Transformation|SetActorRotation (Math|Rotator|MakeRotfromZ (- (Transformation|GetActorLocation (Variables|Default|GetBuildingProp)) (Transformation|GetActorLocation))))
      (Transformation|SetActorRotation (Math|Rotator|MakeRotfromXZ (Math|Vector|Lerp(Vector) (Transformation|GetActorRightVector) (* (Transformation|GetActorForwardVector) (Math|Float|Sign(Float) (Variables|Default|GetRotation))) (Math|Float|Absolute(Float) (Variables|Default|GetRotation))) (Transformation|GetActorUpVector)))
      (Transformation|SetActorRotation (Variables|Default|GetBuildingProp) (Transformation|GetActorRotation) false)
      (Transformation|SetWorldScale3D (Variables|Default|GetSM_Rope) (Math|Conversions|ToVector(Float) (/ (Math|Vector|Distance(Vector) (Transformation|GetActorLocation (Variables|Default|GetBuildingProp)) (Transformation|GetActorLocation)) 1e+02)))
      (Rendering|Material|GetBaseMaterial (Rendering|Material|GetMaterial _staticmeshcomponent 0))
      (bind _returnvalue (Rendering|Material|GetVectorParameterValue _material_instance_dynamic "Wind Speed" :!class /Script/Engine.MaterialInstanceDynamic))
      (bind (_returnvalue_r _returnvalue_g _returnvalue_b) (Math|Color|BreakColor _returnvalue))
      (Rendering|Material|SetCustomPrimitiveDataVector3 _sm_rope 0 (Math|Vector|MakeVector _returnvalue_r _returnvalue_g _returnvalue_b))
      (Variables|SetWindSpeed_X (Math|Vector|MakeVector _returnvalue_r _returnvalue_g _returnvalue_b))
      (bind _returnvalue_2 (Rendering|Material|GetScalarParameterValue _material_instance_dynamic "Twist Intensity" :!class /Script/Engine.MaterialInstanceDynamic))
      (Rendering|Material|SetCustomPrimitiveDataFloat _sm_rope 4 _returnvalue_2)
      (Variables|Default|SetTwistIntensity _returnvalue_2)
      (bind _returnvalue_3 (Rendering|Material|GetScalarParameterValue _material_instance_dynamic "Wobble Intensity" :!class /Script/Engine.MaterialInstanceDynamic))
      (Rendering|Material|SetCustomPrimitiveDataFloat _sm_rope 5 _returnvalue_3)
      (Variables|Default|SetWobbleIntensity _returnvalue_3)
      (bind _returnvalue_4 (Rendering|Material|GetScalarParameterValue _material_instance_dynamic "Pull Amount" :!class /Script/Engine.MaterialInstanceDynamic))
      (Rendering|Material|SetCustomPrimitiveDataFloat _sm_rope 6 _returnvalue_4)
      (Variables|Default|SetPullAmount _returnvalue_4)
      (Rendering|Material|SetCustomPrimitiveDataFloat _sm_rope 7 3.5)
      (Rendering|Material|SetCustomPrimitiveDataVector3 (Variables|/Script/FortniteGame.BuildingSMActor|GetStaticMeshComponent (Variables|Default|GetBuildingProp)) 0 (Transformation|GetActorLocation)))
    (:else)))
