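# Material function animating wind-driven kites through World Position Offset and matching normals, with static switches for rope, fan and fish variants.
# Panned wind noise, storm weight and per-primitive randomized time drive twist, wobble, tension pull and cloth flutter.
# Ropes get parabolic slack from length versus pulled distance; fish variants wobble fins by position masks.
# Offsets fade with distance via a Nanite-specific branch; ripple normals and subsurface modulation add cloth detail.
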MaterialFunction /Game/ArtTools/Tech_Art/KiteAutoWireTool/Materials/MaterialFunctions/MF_KiteMovement
graph {nodes [
  Constant3Vector()
  n1 = Frac(None=TransformPosition(None=NamedRerouteUsage()) {transformType=TRANSFORMPOSSOURCE_World})
  n2 = MaterialFunctionCall(Float3=@n1) {materialFunction=&BreakOutFloat3Components}
  Frac(None=Add(A=Add(A=@n2.R, B=@n2.R), B=@n2.B))
  n3 = GetMaterialAttributes(None=FunctionInput() {inputName=Material, inputType=FunctionInput_MaterialAttributes, sortPriority=-1}) {attributeGetTypes [#1, #2, #3]}
  n4 = NamedRerouteDeclaration(None=@n3.WorldPositionOffset) {name=OriginalWPO, nodeColor=<rgba 1 0.9141 0 1>}
  n5 = NamedRerouteUsage()
  n6 = PixelDepth()
  n7 = ScalarParameter() {defaultValue=20000, group=WPO Fade, parameterName=WPO_Fade_Offset}
  n8 = ScalarParameter() {defaultValue=10000, group=WPO Fade, parameterName=WPO_Fade_Distance}
  n9 = NaniteReplace!(
   Default=Divide(A=Subtract(A=@n6.R, B=@n7), B=@n8)
   Nanite=MaterialFunctionCall!(FadeBlendDistance=@n8, AutoHandleBadDistance) {materialFunction=/Game/Global/MaterialFunctions/TechArt/MF_WPODisableBlend}.VertexShaderOutput
  )
  n10 = NamedRerouteUsage()
  n11 = MaterialFunctionCall(Rotation Axis=NamedRerouteUsage(), Rotation Angle=NamedRerouteUsage()) {materialFunction=&FixRotateAboutAxisNormals}
  n12 = MaterialFunctionCall(
   BaseNormal=StaticSwitchParameter(
    True=@n10
    False=MaterialFunctionCall(BaseNormal=@n10, AdditionalNormal=@n11.New Tangent Space Vertex Normal) {materialFunction=&BlendAngleCorrectedNormals}.Result
   ) {parameterName=Use Rope, sortPriority=2}
   AdditionalNormal=MaterialFunctionCall(Rotation Axis=NamedRerouteUsage(), Rotation Angle=NamedRerouteUsage()) {materialFunction=&FixRotateAboutAxisNormals}.New Tangent Space Vertex Normal
  ) {materialFunction=&BlendAngleCorrectedNormals}
  n13 = MaterialFunctionCall(
   Normal=StaticSwitchParameter(True=NamedRerouteUsage(), False=NamedRerouteUsage()) {group=Kite | Ripples, parameterName=Use Triplanar Ripples}
   Flatness=LinearInterpolate(B=ScalarParameter() {defaultValue=0.1, parameterName=Flatten Ripples Normals}, Alpha=NamedRerouteUsage()) {constA=1}
  ) {materialFunction=&FlattenNormal}
  n14 = SetMaterialAttributes!(
   MaterialAttributes=@n3.MaterialAttributes
   WorldPositionOffset=LinearInterpolate(A=Add(A=@n4, B=@n5), Alpha=Saturate(None=@n9)) {constB=0}
   Normal=StaticSwitchParameter(
    True=@n12.Result
    False=MaterialFunctionCall(BaseNormal=@n12.Result, AdditionalNormal=@n13.Result) {materialFunction=&BlendAngleCorrectedNormals}.Result
   ) {parameterName=Use Rope, sortPriority=2}
   EmissiveColor
   SubsurfaceColor=Multiply(A=@n3.SubsurfaceColor, B=NamedRerouteUsage())
  ) {attributeSetTypes [
    #1
    #2
    #4
    #3
   ]}
  FunctionOutput(None=StaticSwitchParameter(True=@n14, False=@n3.MaterialAttributes) {parameterName=Use Kite Wind, sortPriority=1}) {sortPriority=10}
  MaterialFunctionCall() {materialFunction=&DebugTimeSine}
  n15 = MaterialFunctionCall!(Position, IsTrees, IsGrass, IsMediumPlant, IsMediumPlant_2024Anim, IsMediumPlant2ndLevel, MediumPlant_BasePivotPostion) {materialFunction=/Game/Environments/Asteria/Foliage/Wind/MF_WindActorData}
  Max(A=@n15.StormMask, B=ScalarParameter() {parameterName=Storm-Gust Debug})
  n16 = NamedRerouteUsage()
  n17 = Multiply(A=Add(A=NamedRerouteUsage(), B=NamedRerouteUsage()), B=Saturate(None=Length(None=@n16) {is=class MaterialExpressionLength}))
  n18 = StaticSwitchParameter(True=Add(A=NamedRerouteUsage(), B=@n17), False=@n17) {parameterName=Use Fan, sortPriority=2}
  n19 = Add(A=NamedRerouteUsage(), B=@n18)
  n20 = Transform(None=Add(A=StaticSwitchParameter(True=@n19, False=@n18) {parameterName=Use Fish, sortPriority=2}, B=NamedRerouteUsage())) {transformSourceType=TRANSFORMSOURCE_Local}
  n21 = NamedRerouteUsage()
  NamedRerouteDeclaration(None=Add(A=Add(A=@n20, B=@n21), B=NamedRerouteUsage()) {constB=0}) {name=Kite WPO, nodeColor=<rgba 0 1 0.2969 1>}
  n22 = NamedRerouteUsage()
  n23 = ScalarParameter() {defaultValue=0.2, parameterName=Twist Wiggle}
  n24 = ScalarParameter() {defaultValue=1, parameterName=Storm Wind Twist Wiggle}
  n25 = NamedRerouteUsage()
  n26 = Normalize(VectorInput=LinearInterpolate(A=@n22, B=NamedRerouteUsage(), Alpha=LinearInterpolate(A=@n23, B=@n24, Alpha=@n25)) {constAlpha=0})
  n27 = ScalarParameter() {defaultValue=0.1, parameterName=Twist Intensity, sortPriority=100}
  n28 = NamedRerouteDeclaration(None=Multiply(
    A=NamedRerouteUsage()
    B=FunctionInput(Preview=@n27) {bUsePreviewValueAsDefault=true, inputName=Twist Intensity, inputType=FunctionInput_Scalar, sortPriority=50}
   )) {name=Twist Angle, nodeColor=<rgba 0 0.3672 1 1>}
  n29 = Constant3Vector()
  n30 = Add(A=LocalPosition().XYZ, B=NamedRerouteUsage())
  NamedRerouteDeclaration(None=StaticSwitchParameter(True=Constant(), False=RotateAboutAxis(NormalizedRotationAxis=@n26, RotationAngle=@n28, PivotPoint=@n29, Position=@n30)) {parameterName=Use Rope, sortPriority=2}) {name=O1 - Twist WPO, nodeColor=<rgba 0.4531 0 1 1>}
  n31 = Time()
  n32 = ScalarParameter() {bUseCustomPrimitiveData=true, parameterName=Random, primitiveDataIndex=14}
  n33 = LinearInterpolate(Alpha=@n32) {constA=0.99, constB=1.01}
  n34 = NamedRerouteDeclaration(None=Add(A=Multiply(A=@n31, B=@n33), B=Multiply(A=@n32) {constB=120})) {name=Random Time, nodeColor=<rgba 0 0.9062 1 1>}
  n35 = Sine(None=Divide(A=@n34) {constB=20})
  n36 = SmoothStep(Value=Multiply(A=ConstantBiasScale(None=@n35)) {constB=15}) {constMax=3}
  n37 = NamedRerouteUsage()
  NamedRerouteDeclaration(None=Max(A=Max(A=@n36, B=@n37), B=ScalarParameter() {bUseCustomPrimitiveData=true, parameterName=AlwaysInTention, primitiveDataIndex=15})) {name=Tention, nodeColor=<rgba 0 0.3437 1 1>}
  n38 = ComponentMask(None=TextureCoordinate()) {desc=t, g=true}
  n39 = NamedRerouteDeclaration(None=FunctionInput(Preview=ScalarParameter() {defaultValue=150, parameterName=Pull Amount, sortPriority=150}) {bUsePreviewValueAsDefault=true, inputName=Pull Amount, inputType=FunctionInput_Scalar, sortPriority=70}) {name=Pull Amount, nodeColor=<rgba 1 0.9141 0 1>}
  n40 = MaterialFunctionCall() {materialFunction=&ObjectScale}
  n41 = Multiply(A=OneMinus(None=NamedRerouteUsage()), B=Multiply(A=NamedRerouteUsage(), B=Divide(A=@n39, B=@n40.Scale XYZ)))
  NamedRerouteDeclaration(None=StaticSwitchParameter(True=Multiply(A=@n38, B=@n41), False=@n41) {parameterName=Use Rope, sortPriority=2}) {name=Pull WPO, nodeColor=<rgba 0.6641 1 0 1>}
  n42 = FunctionInput() {bUsePreviewValueAsDefault=true, inputName=Misc Masks, inputType=FunctionInput_Vector4, sortPriority=300, previewValue=<1 1 1 1>}
  NamedRerouteDeclaration(None=ChannelMaskParameter(None=@n42) {parameterName=Medium Frequency Mask Channel}.R) {name=Med Freq Mask, nodeColor=<rgba 1 0 0.75 1>}
  NamedRerouteDeclaration(None=VectorParameter() {parameterName=Forward Axis, defaultValue=<rgba 0 1 0 0>}.RGB) {name=Forward Axis, nodeColor=<rgba 0.2 1 0.2 1>}
  n43 = Divide(A=LocalPosition().XY, B=ScalarParameter() {defaultValue=700, parameterName=Medium Frequency Wind Scale}) {constB=1000}
  n44 = NamedRerouteUsage()
  n45 = ScalarParameter() {defaultValue=0.19, parameterName=Medium Frequency Wind Speed}
  n46 = Constant4Vector() {constant=<rgba 1.343 2.1317 -2.211 -1.1713>}
  n47 = Panner(Coordinate=@n43, Time=@n44, Speed=Multiply(A=@n45, B=AppendVector(A=@n46.R, B=@n46.R)))
  n48 = TextureObject() {samplerType=SAMPLERTYPE_Normal, texture=/Game/Global/Textures/Default/Noise/T_WindNoise}
  n49 = Panner(Coordinate=@n43, Time=@n44, Speed=Multiply(A=@n45, B=AppendVector(A=@n46.B, B=@n46.B)))
  NamedRerouteDeclaration(None=LinearInterpolate(A=TextureSample(UVs=@n47, Tex=@n48).RGB, B=TextureSample(UVs=@n49, Tex=@n48).RGB)) {name=Normals For WPO, nodeColor=<rgba 0.875 0 1 1>}
  n50 = NamedRerouteUsage()
  n51 = NamedRerouteDeclaration(None=Multiply(A=Time(), B=ScalarParameter() {defaultValue=0.15, parameterName=Fan Speed})) {name=Fan Speed, nodeColor=<rgba 0 0.3672 1 1>}
  n52 = RotateAboutAxis(NormalizedRotationAxis=@n50, RotationAngle=@n51, PivotPoint=Constant(), Position=NamedRerouteUsage())
  n53 = NamedRerouteUsage()
  NamedRerouteDeclaration(None=StaticSwitchParameter(True=Add(A=@n52, B=@n53), False=@n53) {parameterName=Use Fan, sortPriority=2}) {name=Up Axis Fixed, nodeColor=<rgba 0.7 0.7 1 1>}
  n54 = Multiply(A=NamedRerouteUsage(), B=ScalarParameter() {defaultValue=0.1, parameterName=Orbital Period})
  n55 = Sine(None=@n54)
  n56 = Cosine(None=@n54)
  n57 = NamedRerouteDeclaration(None=Normalize(VectorInput=MaterialFunctionCall(X=@n55, Y=@n56) {materialFunction=&MakeFloat3}.Result)) {name=Wobble Axis, nodeColor=<rgba 0.6875 1 0 1>}
  n58 = ScalarParameter() {defaultValue=0.02, parameterName=Wobble Intensity, sortPriority=101}
  n59 = NamedRerouteDeclaration(None=Multiply(
    A=NamedRerouteUsage()
    B=FunctionInput(Preview=@n58) {bUsePreviewValueAsDefault=true, inputName=Wobble Intensity, inputType=FunctionInput_Scalar, sortPriority=60}
   )) {name=Wobble Angle, nodeColor=<rgba 0.9922 0 1 1>}
  n60 = Constant()
  n61 = Constant3Vector() {constant=<rgba 0 0 -1 0>}
  NamedRerouteDeclaration(None=Add(A=RotateAboutAxis(NormalizedRotationAxis=@n57, RotationAngle=@n59, PivotPoint=@n60, Position=@n61), B=@n61)) {name=Pull Vector, nodeColor=<rgba 1 0 0.5859 1>}
  n62 = NamedRerouteDeclaration(None=FunctionInput(Preview=VectorParameter() {bUseCustomPrimitiveData=true, parameterName=Pivot}.RGB) {bUsePreviewValueAsDefault=true, inputName=Custom Pivot, sortPriority=30}) {name=Pivot, nodeColor=<rgba 0 1 0.2031 1>}
  n63 = LocalPosition()
  n64 = Add(A=NamedRerouteUsage(), B=NamedRerouteUsage())
  NamedRerouteDeclaration(None=RotateAboutAxis(
    NormalizedRotationAxis=@n57
    RotationAngle=@n59
    PivotPoint=TransformPosition(None=@n62) {transformSourceType=TRANSFORMPOSSOURCE_World}
    Position=Add(A=@n63.XYZ, B=@n64)
   )) {name=O2 - Wobble WPO, nodeColor=<rgba 0 1 0.1328 1>}
  n65 = Constant3Vector() {constant=<rgba 0 0 -1 0>}
  n66 = MaterialFunctionCall() {materialFunction=&ObjectScale}
  n67 = Multiply(A=@n66.Scale X) {constB=100}
  n68 = Multiply(A=NamedRerouteUsage(), B=OneMinus(None=NamedRerouteUsage()))
  n69 = Subtract(
   A=Power(Base=Multiply(A=@n66.Scale X) {constB=101, desc=Rope Length}) {desc=Rope Length}
   B=Power(Base=Subtract(A=@n67, B=@n68)) {desc=Straight Distance}
  )
  n70 = ComponentMask(None=TextureCoordinate()) {desc=T, g=true}
  n71 = Multiply(
   A=Multiply(A=ScalarParameter() {defaultValue=1.5, parameterName=Slack Ratio}, B=SquareRoot(None=@n69)) {constA=4}
   B=Multiply(A=@n70, B=OneMinus(None=@n70))
  )
  NamedRerouteDeclaration(None=StaticSwitchParameter(True=Multiply(A=@n65, B=@n71), False=Constant()) {parameterName=Use Rope, sortPriority=2}) {name=Slack WPO, nodeColor=<rgba 0.6406 1 0 1>}
  NamedRerouteDeclaration(None=FunctionInput() {bUsePreviewValueAsDefault=true, inputName=Cloth Mask, inputType=FunctionInput_Scalar, sortPriority=200}) {name=Cloth Mask, nodeColor=<rgba 0 1 0.5547 1>}
  n72 = MaterialFunctionCall(Color=NamedRerouteUsage()) {materialFunction=&SplitComponents}
  n73 = Transform(None=MaterialFunctionCall(X=@n72.R, Y=@n72.R, Z=Subtract(A=@n72.B)) {materialFunction=&MakeFloat3}.Result) {transformSourceType=TRANSFORMSOURCE_Local}
  n74 = Multiply(A=NamedRerouteUsage(), B=NamedRerouteUsage())
  NamedRerouteDeclaration(None=Multiply(A=Multiply(A=@n73, B=@n74), B=ScalarParameter() {defaultValue=20, parameterName=Mid Freq Amp})) {name=Medium Freq WPO, nodeColor=<rgba 1 0 0.7031 1>}
  n75 = DotProduct(
   A=StaticSwitchParameter(True=NamedRerouteUsage(), False=NamedRerouteUsage()) {group=Kite | Ripples, parameterName=Use Triplanar Ripples}
   B=Constant() {r=1}
  )
  n76 = ScalarParameter() {defaultValue=0.3, parameterName=SSS Fake Normals}
  NamedRerouteDeclaration(None=StaticSwitchParameter(True=Constant() {r=1}, False=LinearInterpolate(B=@n75, Alpha=@n76) {constA=1}) {parameterName=Use Rope, sortPriority=2}) {name=SSS Multiplier, nodeColor=<rgba 1 0 0.3281 1>}
  n77 = Constant()
  n78 = Multiply(A=Constant2Vector() {g=1.17, r=1.13}, B=ScalarParameter() {defaultValue=0.15, parameterName=High Freq Speed})
  NamedRerouteDeclaration(None=TextureSample(UVs=Panner(Coordinate=@n77, Time=@n44, Speed=@n78), Tex=@n48).RGB) {name=High Freq WPO Norm, nodeColor=<rgba 0.5234 1 0 1>}
  n79 = VectorParameter() {parameterName=PivotPointBias}
  n80 = LocalPosition()
  NamedRerouteDeclaration(None=StaticSwitchParameter(True=RotateAboutAxis(NormalizedRotationAxis=@n50, RotationAngle=@n51, PivotPoint=@n79.RGB, Position=@n80.XYZ), False=Constant()) {parameterName=Use Fan, sortPriority=2}) {name=Fan Motion WPO, nodeColor=<rgba 1 0 0.4687 1>}
  n81 = Divide(A=LocalPosition().XY, B=ScalarParameter() {defaultValue=1400, group=Kite | Ripples, parameterName=Ripples Scale}) {constB=1000}
  n82 = TextureSample(UVs=@n81) {samplerType=SAMPLERTYPE_LinearColor, texture=/Game/Effects/Fort_Effects/Textures/Plasma/T_Turbulence_AdditiveWarp_Linear}
  n83 = Add(
   A=@n81
   B=Multiply(A=ComponentMask(None=@n82.RGB) {g=true, r=true}, B=ScalarParameter() {defaultValue=0.05, group=Kite | Ripples, parameterName=Ripples Warp})
  )
  n84 = ScalarParameter() {defaultValue=0.3, group=Kite | Ripples, parameterName=Ripples Wind Speed}
  n85 = Constant4Vector() {constant=<rgba 0.4521 1.8894 -0.4455 1.4716>}
  n86 = AppendVector(A=@n85.R, B=@n85.R)
  n87 = NamedRerouteDeclaration(None=TextureObject() {samplerType=SAMPLERTYPE_Normal, texture=/Game/Global/Textures/Default/Noise/BentNoise_N}) {name=Bend T Object, nodeColor=<rgba 0.8047 0 1 1>}
  n88 = TextureSample(UVs=Panner(Coordinate=@n83, Time=NamedRerouteUsage(), Speed=Multiply(A=@n84, B=@n86)), Tex=@n87)
  n89 = AppendVector(A=@n85.B, B=@n85.B)
  n90 = TextureSample(UVs=Panner(Coordinate=@n83, Time=NamedRerouteUsage(), Speed=Multiply(A=@n84, B=@n89)), Tex=@n87)
  n91 = Add(A=@n88.RGB, B=@n90.RGB)
  NamedRerouteDeclaration(None=AppendVector(A=ComponentMask(None=@n91) {g=true, r=true}, B=Multiply(A=@n88.B, B=@n90.B))) {name=Ripples Normals 2D, nodeColor=<rgba 1 0.3516 0 1>}
  NamedRerouteDeclaration(None=VectorParameter() {parameterName=Up Axis, sortPriority=20, defaultValue=<rgba 0 0 1 0>}.RGB) {name=Up Axis, nodeColor=<rgba 0.2 0.2 1 1>}
  n92 = NamedRerouteUsage()
  n93 = ScalarParameter() {defaultValue=1400, group=Kite | Ripples, parameterName=Ripples Scale}
  n94 = MaterialFunctionCall(X=@n93, Y=@n93, Z=ScalarParameter() {defaultValue=5000, group=Kite | Ripples, parameterName=Ripples Scale Z}) {materialFunction=&MakeFloat3}
  n95 = Transform(None=Constant3Vector() {constant=<rgba 0 0 1 0>}) {transformType=TRANSFORM_Local}
  n96 = ScalarParameter() {defaultValue=1400, group=Kite | Ripples, parameterName=Ripples Scale}
  n97 = ScalarParameter() {defaultValue=0.3, group=Kite | Ripples, parameterName=Ripples Wind Speed}
  n98 = Add(A=LocalPosition() {includedOffsets=ExcludeOffsets}.XYZ, B=Multiply(A=NamedRerouteUsage(), B=Multiply(A=@n96, B=@n97)))
  n99 = NamedRerouteUsage()
  NamedRerouteDeclaration(None=LinearInterpolate(
    A=MaterialFunctionCall(TextureObject=@n92, TextureSize=@n98, WorldPosition=@n94.Result, World Space Normal=@n95) {materialFunction=&WorldAlignedTexture}.XYZ Texture
    B=MaterialFunctionCall(TextureObject=@n99, TextureSize=@n94.Result, WorldPosition=@n98, World Space Normal=@n95) {materialFunction=&WorldAlignedTexture}.XYZ Texture
   )) {name=Ripples Normals 3D, nodeColor=<rgba 0.0312 0 1 1>}
  NamedRerouteDeclaration(None=ChannelMaskParameter(None=@n42) {maskChannel=Green, parameterName=Random Offset Mask Channel, defaultValue=<rgba 0 1 0 0>}.R) {name=Mask Random Offset, nodeColor=<rgba 1 0 0.3516 1>}
  n100 = NamedRerouteUsage()
  n101 = Constant4Vector() {constant=<rgba 21.1 13.2 15.7 11.5>}
  n102 = Multiply(A=NamedRerouteUsage(), B=@n101.R)
  n103 = Multiply(A=NamedRerouteUsage(), B=@n101.G)
  n104 = Multiply(A=NamedRerouteUsage(), B=@n101.B)
  n105 = Multiply(A=NamedRerouteUsage(), B=@n101.A)
  n106 = NamedRerouteDeclaration(None=Add(A=Add(A=@n102, B=@n103), B=Add(A=@n104, B=@n105))) {name=Fish Random, nodeColor=<rgba 1 0.8906 0 1>}
  n107 = NamedRerouteUsage()
  n108 = Multiply(
   A=Add(A=NamedRerouteUsage(), B=LinearInterpolate(A=@n106, Alpha=@n107))
   B=ScalarParameter() {defaultValue=0.5, parameterName=Fish Wobble Freq}
  )
  n109 = LinearInterpolate(A=Cosine(None=@n108) {period=1.12}, B=Sine(None=@n108) {period=1.12}, Alpha=NamedRerouteUsage())
  n110 = Multiply(A=Add(A=NamedRerouteUsage(), B=NamedRerouteUsage()), B=ScalarParameter() {defaultValue=0.1, parameterName=Fish Fins Phasing Speed})
  n111 = Multiply(
   A=ComponentMask(None=MaterialFunctionCall!(UV1=@n110, Min, Max) {materialFunction=/Game/Athena/WinterArt/Materials/MF_SimpleWindNoise}.Result) {r=true}
   B=LinearInterpolate(
    A=ScalarParameter() {defaultValue=20, parameterName=Fish Mini Wobble Amp}
    B=ScalarParameter() {defaultValue=50, parameterName=Fish Wobble Amp}
    Alpha=NamedRerouteUsage()
   )
  )
  n112 = LocalPosition() {includedOffsets=ExcludeOffsets}
  n113 = Sign(None=ComponentMask(None=@n112.XYZ) {r=true})
  n114 = Constant() {r=1}
  n115 = LocalPosition() {includedOffsets=ExcludeOffsets}
  n116 = Constant() {r=1}
  n117 = NamedRerouteDeclaration(None=If(A=ComponentMask(None=@n115.XYZ) {g=true}, "A > B"=Constant(), "A == B"=@n116, "A < B"=@n116) {constB=-200, desc=Back}) {name=Fish Back Fin, nodeColor=<rgba 0.2891 0 1 1>}
  n118 = Constant()
  n119 = NamedRerouteDeclaration(None=If(A=@n115.Z, "A > B"=Constant() {r=1}, "A == B"=@n118, "A < B"=@n118) {constB=450, desc=Top}) {name=Fish Top Fin, nodeColor=<rgba 0.0312 1 0 1>}
  n120 = LinearInterpolate(
   A=Normalize(VectorInput=MaterialFunctionCall(X=@n113, Z=@n114) {materialFunction=&MakeFloat3}.Result)
   B=Constant3Vector() {constant=<rgba 1 0 0 0>}
   Alpha=NamedRerouteDeclaration(None=Max(A=@n117, B=@n119)) {name=Main Fins, nodeColor=<rgba 0 1 0.9063 1>}
  )
  NamedRerouteDeclaration(None=Multiply(A=Multiply(A=@n100, B=@n109), B=Multiply(A=@n111, B=@n120) {constB=20}) {constB=20}) {name=Fish WPO, nodeColor=<rgba 0 0.6719 1 1>}
  NamedRerouteDeclaration(None=@n1) {name=Pivot Frac, nodeColor=<rgba 1 0.3984 0 1>}
  n121 = Sign(None=@n112.XYZ)
  n122 = DotProduct(A=Constant2Vector() {g=1, r=1}, B=@n121)
  NamedRerouteDeclaration(None=Saturate(None=@n122)) {name=Fin Mask 1, nodeColor=<rgba 1 0.1875 0 1>}
  n123 = Constant()
  n124 = FunctionInput(Preview=ComponentMask(None=VectorParameter() {parameterName=Wind Speed, defaultValue=<rgba 0.025 0.01 0 0>}.RGB) {g=true, r=true}) {bUsePreviewValueAsDefault=true, inputName=Wind Speed, inputType=FunctionInput_Vector2, sortPriority=30}
  n125 = Panner(Coordinate=@n123, Time=@n34, Speed=@n124) {speedX=0.01}
  n126 = ScalarParameter() {defaultValue=6, parameterName=Storm Wind Speed Multiplier}
  n127 = MaterialFunctionCall(UV1=Panner(Coordinate=@n123, Time=@n34, Speed=Multiply(A=@n126, B=@n124)) {speedX=0.01}) {materialFunction=&MF_SimpleWindNoise}
  n128 = LinearInterpolate(
   A=ConstantBiasScale(None=MaterialFunctionCall(UV1=@n125) {materialFunction=&MF_SimpleWindNoise}.Result) {bias=-0.5, scale=2}
   B=Multiply(
    A=ConstantBiasScale(None=@n127.Result) {bias=-0.5, scale=2}
    B=ScalarParameter() {defaultValue=0.2, parameterName=Storm Wind Movement Multiplier}
   )
   Alpha=NamedRerouteUsage()
  )
  NamedRerouteDeclaration(None=ComponentMask(None=@n128) {g=true}) {name=Wind 2, nodeColor=<rgba 0 0.5078 1 1>}
  NamedRerouteDeclaration(None=Saturate(None=Multiply(A=@n122) {constB=-1})) {name=Fin Mask 2, nodeColor=<rgba 1 0.1875 0 1>}
  n129 = DotProduct(A=Constant2Vector() {g=-1, r=1}, B=@n121)
  NamedRerouteDeclaration(None=Saturate(None=@n129)) {name=Fin Mask 3, nodeColor=<rgba 1 0.1875 0 1>}
  NamedRerouteDeclaration(None=Saturate(None=Multiply(A=@n129) {constB=-1})) {name=Fin Mask 4, nodeColor=<rgba 1 0.1875 0 1>}
  NamedRerouteDeclaration(None=@n15.StormMask) {name=Storm-Gust Weight, nodeColor=<rgba 0.5938 0 1 1>}
  NamedRerouteDeclaration(None=ComponentMask(None=@n128) {r=true}) {name=Wind 1, nodeColor=<rgba 0.8281 1 0 1>}
  NamedRerouteDeclaration(None=@n3.Normal) {name=Original Normal, nodeColor=<rgba 0 1 0.6719 1>}
  NamedRerouteUsage()
  Reroute(None=@n22)
 ]}
