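# Parent material for kites in the Kite Auto Wire Tool, with animated movement and translucent-looking cloth, enabled for Nanite.
# MF_GigaMix_2023 supplies the base attributes, blended by a mask toward a Two Sided Foliage variant with tinted base color as subsurface.
# The mask, a selectable SFX_Mask channel, also serves as cloth mask for MF_KiteMovement, alongside a WPO mask texture.
# The result then passes through MF_CreativeMoveTool.
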Material /Game/ArtTools/Tech_Art/KiteAutoWireTool/Materials/M_Parent_Kite
graph {
 properties {
  bAutomaticallySetUsageInEditor=false
  bUseMaterialAttributes=true
  bUsedWithNanite=true
  shadingModel=MSM_FromMaterialExpression
  usedShadingModels=Default Lit | Subsurface | Clear Coat | Two Sided Foliage
 }
 nodes [
  n1 = NamedRerouteUsage()
  n2 = VectorParameter() {parameterName=SSS Tint, defaultValue=<rgba 1 1 1 1>}
  n3 = MaterialFunctionCall() {materialFunction=/Game/Global/Base4Layers_2022/MaterialFunctions/MF_GigaMix_2023}
  n4 = SetMaterialAttributes!(
   MaterialAttributes=@n3.Result
   Opacity=Constant() {r=0.3}
   Subsurface Color=LinearInterpolate(A=Multiply(A=@n1, B=@n2.RGB), Alpha=ScalarParameter() {parameterName=SSS Opacity}) {constAlpha=0, constB=0}
   Shading Model=ShadingModel() {shadingModel=MSM_TwoSidedFoliage}
  ) {attributeSetTypes [#1, #2, #3]}
  n5 = NamedRerouteUsage()
  MaterialFunctionCall(MA In=MaterialFunctionCall!(
    Material=BlendMaterialAttributes(A=@n3.Result, B=@n4, Alpha=@n5)
    Wind Speed
    Custom Pivot
    Twist Intensity
    Wobble Intensity
    Pull Amount
    Cloth Mask=@n5
    Misc Masks=TextureSampleParameter2D() {parameterName=WPO Mask, samplerType=SAMPLERTYPE_Masks, texture=&T_WhiteMask}.RGBA
   ) {materialFunction=/Game/ArtTools/Tech_Art/KiteAutoWireTool/Materials/MaterialFunctions/MF_KiteMovement}.Result) {materialFunction=&MF_CreativeMoveTool}
  NamedRerouteDeclaration(None=GetMaterialAttributes(None=@n3.Result) {attributeGetTypes [#4]}.Base Color) {name=SSS, nodeColor=<rgba 0 1 0.1094 1>}
  n6 = TextureObjectParameter() {
   desc="SFX UniversalTextureMask Uses Dot to select channel used for various parts of material.You can reuse that texture across layers and functions. "
   group=0_SFX Global
   parameterName=SFX_Mask
   samplerType=SAMPLERTYPE_Masks
   sortPriority=0
   texture=/Game/Global/Base4Layers_2022/Textures/T_SFX_M
  }
  NamedRerouteDeclaration(None=ChannelMaskParameter(None=TextureSample(Tex=@n6) {samplerType=SAMPLERTYPE_Masks}.RGBA) {parameterName=SSS Mask}.R) {name=SSS Mask, nodeColor=<rgba 1 0 0.3047 1>}
 ]
}
