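# Parent rope material for the kite auto wire tool, a two-sided lit surface for spline meshes.
# MF_KiteMovement drives World Position Offset from wind, wobble, twist and pull values, mostly per-primitive custom data.
# SplineThicken sets rope thickness and normals, with an offset cancelling the mesh's original width.
# Tileable UVs derive length from distance to the object origin, and one mask channel feeds specular, metallic and roughness.
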Material /Game/ArtTools/Tech_Art/KiteAutoWireTool/Materials/M_Parent_Rope
graph {
 properties {
  bAutomaticallySetUsageInEditor=false
  bHasPixelAnimation=true
  bTangentSpaceNormal=false
  bUseMaterialAttributes=true
  bUsedWithSplineMeshes=true
  twoSided=true
  usedShadingModels=Default Lit
 }
 nodes [
  MaterialFunctionCall() {materialFunction=&ObjectScale}
  n1 = Distance(
   A=WorldPosition() {worldPositionShaderOffset=WPT_CameraRelative}.XYZ
   B=TransformPosition(None=Constant3Vector()) {transformType=TRANSFORMPOSSOURCE_TranslatedWorld}
  )
  Multiply(B=@n1) {constA=100}
  n2 = ScalarParameter() {bUseCustomPrimitiveData=true, defaultValue=3.5, parameterName=SplineThickness, primitiveDataIndex=7}
  n3 = If(A=@n2, "A > B"=@n2, "A == B"=Constant() {r=3.5}, "A < B"=@n2)
  n4 = Divide(A=@n1) {constB=100}
  n5 = VectorParameter() {parameterName=UVs, channelNames {a=Shift Y, b=Shift X, g=Tiling Y, r=Tiling X}, defaultValue=<rgba 1 0.1 0 0.21>}
  n6 = ComponentMask(None=TextureCoordinate()) {r=true}
  n7 = AppendVector(A=Multiply(A=Add(A=@n4, B=@n5.Shift X), B=@n5.Tiling X), B=Multiply(A=Add(A=@n5.Shift Y, B=@n6), B=@n5.Tiling Y))
  n8 = NamedRerouteUsage()
  n9 = TextureSampleParameter2D(UVs=@n8) {
   parameterName=Normal
   samplerType=SAMPLERTYPE_Normal
   texture=/Game/ArtTools/Tech_Art/KiteAutoWireTool/CozySoup_KiteAnchor_A/Textures/T_CozySoup_KiteAnchor_A_A_N
  }
  n10 = MaterialFunctionCall(
   WidthBase=@n3
   WidthTip=@n3
   WorldPosition=Add(A=WorldPosition().XYZ, B=NamedRerouteUsage())
   UVs for texturing=@n7
   AdditionalNormal=@n9.RGB
  ) {materialFunction=&SplineThicken}
  n11 = LocalPosition()
  n12 = Multiply(A=Sign(None=ComponentMask(None=@n11.XYZ) {r=true}), B=VectorParameter() {parameterName=Original Width, defaultValue=<rgba 0.25 0 0 1>}.RGB)
  NamedRerouteDeclaration(None=Subtract(A=@n10.WorldPositionOffset, B=Transform(None=@n12) {transformSourceType=TRANSFORMSOURCE_Local}) {constA=0}) {name=WPO Out, nodeColor=<rgba 0 0.8828 1 1>}
  NamedRerouteDeclaration(None=@n10.UVs with Parallax) {name=UV Parallaxed, nodeColor=<rgba 1 0.0703 0 1>}
  NamedRerouteDeclaration(None=AppendVector(A=@n7, B=Constant())) {name=UV Deb, nodeColor=<rgba 1 0.6094 0 1>}
  NamedRerouteDeclaration(None=@n7) {name=UVs from Pos, nodeColor=<rgba 1 0 0.2109 1>}
  NamedRerouteUsage()
  NamedRerouteUsage()
  n13 = VectorParameter() {bUseCustomPrimitiveData=true, parameterName=Wind Speed, sortPriority=50, channelNames {a="", b="", g=Wind Y, r=Wind X}}
  n14 = Constant3Vector()
  n15 = ScalarParameter() {bUseCustomPrimitiveData=true, defaultValue=0.02, parameterName=Wobble Intensity, primitiveDataIndex=5, sortPriority=60}
  n16 = GetMaterialAttributes(None=MaterialFunctionCall!(
    Material=SetMaterialAttributes!(MaterialAttributes)
    Wind Speed=ComponentMask(None=@n13) {g=true, r=true}
    Custom Pivot=TransformPosition(None=@n14) {transformSourceType=TRANSFORMPOSSOURCE_Instance, transformType=TRANSFORMPOSSOURCE_World}
    Twist Intensity=ScalarParameter() {bUseCustomPrimitiveData=true, parameterName=Twist Intensity, primitiveDataIndex=4, sortPriority=60}
    Wobble Intensity=@n15
    Pull Amount=ScalarParameter() {bUseCustomPrimitiveData=true, defaultValue=150, parameterName=Pull Amount, primitiveDataIndex=6}
    Cloth Mask=Constant()
    Misc Masks
   ) {materialFunction=/Game/ArtTools/Tech_Art/KiteAutoWireTool/Materials/MaterialFunctions/MF_KiteMovement}.Result) {attributeGetTypes [#1]}
  n17 = TextureSampleParameter2D(UVs=@n8) {
   parameterName=Specular Mask
   samplerType=SAMPLERTYPE_Masks
   texture=/Game/ArtTools/Tech_Art/KiteAutoWireTool/CozySoup_KiteAnchor_A/Textures/T_CozySoup_KiteAnchor_A_A_S
  }
  SetMaterialAttributes!(
   MaterialAttributes=@n16.MaterialAttributes
   Base Color=TextureSampleParameter2D(UVs=@n8) {parameterName=Diffuse, texture=/Game/ArtTools/Tech_Art/KiteAutoWireTool/CozySoup_KiteAnchor_A/Textures/T_CozySoup_KiteAnchor_A_A_D}.RGB
   Specular=@n17.R
   Metallic=@n17.R
   Roughness=@n17.R
   Normal=@n10.Normal
   World Position Offset=Add(A=NamedRerouteDeclaration(None=@n16.World Position Offset) {name=WPO Input, nodeColor=<rgba 0.2422 1 0 1>}, B=NamedRerouteUsage())
   Emissive Color
  ) {attributeSetTypes [
    #2
    #3
    #4
    #5
    #6
    #1
    #7
   ]}
 ]
}
